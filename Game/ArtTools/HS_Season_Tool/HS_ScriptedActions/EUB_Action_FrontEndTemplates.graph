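# Editor utility debug action that targets camera framing bounds on selected FortItemDefinition assets, apparently for front-end pickaxe framing.
# Inside a transaction, it loops over the selection running embedded Python per item, then checks out and saves the assets.
# All bounds and offset inputs share one float getter.
# The Python looks unfinished: its helper is never called, and commented-out offset lines swallow the signature's closing parenthesis.
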EditorUtilityBlueprint /Game/ArtTools/HS_Season_Tool/HS_ScriptedActions/EUB_Action_FrontEndTemplates
graphs [
 DEBUG_ApplyFrontEndTemplate {nodes [
   DEBUG_ApplyFrontEndTemplate() {is=title K2Node_FunctionEntry}
   > BeginTransaction(Context=ApplyFrontEndTemplate)
   > n1 = GetSelectedAssets()
   > n2 = For Each Loop(Array:Array of Object References=@n1) -> (Array Element:Object Reference)
   CheckoutLoadedAssets(execute=@n2.Completed, AssetsToCheckout=@n1)
   > SaveLoadedAssets(AssetsToSave=@n1, bOnlyIfIsDirty=true)
   > EndTransaction()
   n3 = Cast To FortItemDefinition!(execute=@n2.LoopBody, Object=@n2) -> (then, CastFailed, AsFort Item Definition:Fort Item Definition Object Reference) {is=title K2Node_DynamicCast}
   > TransactObject(Object=@n3)
   n4 = Get float() -> (float:double)
   > Execute Python Script!(
    execute
    PythonScript="import unreal\r\n\r\ndef set_pickaxe_camera_properties(asset_path: str,\r\n                                  bounds_origin_x: float,\r\n                                  bounds_origin_y: float,\r\n                                  bounds_origin_z: float,\r\n                                  bounds_half_height: float,\r\n                                  bounds_radius: float,\r\n                                  ##center_offset_x: float,\r\n                                  ##center_offset_y: float,\r\n                                  ##center_offset_z: float):\r\n\r\n    asset = unreal.EditorAssetLibrary.load_asset(asset_path)\r\n    if not asset:\r\n        unreal.log_error(f\"Could not load asset at path: {asset_path}\")\r\n        return\r\n\r\n    unreal.log(f\"Editing properties for: {asset_path}\")\r\n\r\n    bounds = asset.get_editor_property(\"camera_framing_bounds\")\r\n    bounds.origin.x = bounds_origin_x\r\n    bounds.origin.y = bounds_origin_y\r\n    bounds.origin.z = bounds_origin_z\r\n    bounds.cylinder_half_height = bounds_half_height\r\n    bounds.cylinder_radius = bounds_radius\r\n    asset.set_editor_property(\"camera_framing_bounds\", bounds)\r\n\r\n    ##center_offset = asset.get_editor_property(\"camera_framing_bounds_center_offset\")\r\n    ##center_offset.x = center_offset_x\r\n    ##center_offset.y = center_offset_y\r\n    ##center_offset.z = center_offset_z\r\n    ##asset.set_editor_property(\"camera_framing_bounds_center_offset\", center_offset)\r\n\r\n    unreal.EditorAssetLibrary.save_asset(asset_path)\r\n    unreal.log(f\"Updated and saved {asset_path}\")\r\n )\r\n"
    asset_path:String=Get Object Path String(Object=@n3)
    bounds_origin_x:double=@n4
    bounds_origin_y:double=@n4
    bounds_origin_z:double=@n4
    bounds_half_height:double=@n4
    bounds_radius:double=@n4
    center_offset_x:double=@n4
    center_offset_y:double=@n4
    center_offset_z:double=@n4
   ) -> (then, ReturnValue) {inputs [
     asset_path
     bounds_origin_x
     bounds_origin_y
     bounds_origin_z
     bounds_half_height
     bounds_radius
     center_offset_x
     center_offset_y
     center_offset_z
    ]}
  ]}
 EventGraph {links [], nodes []}
]
variables []
