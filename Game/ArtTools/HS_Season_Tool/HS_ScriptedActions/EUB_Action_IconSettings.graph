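# Editor utility action that applies icon texture settings to the assets selected in the editor.
# ApplyIconSettings takes mip generation and texture group inputs, defaulting to no mipmaps and the UI group.
# It loops over the selection, casts each asset to Texture2D, registers it with the transaction and sets both properties.
# After the loop it checks out the whole selection, saves dirty assets and closes the transaction.
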EditorUtilityBlueprint /Game/ArtTools/HS_Season_Tool/HS_ScriptedActions/EUB_Action_IconSettings
graphs [
 ApplyIconSettings {nodes [
   ApplyIconSettings!() -> (then, Mip Gen Settings:TextureMipGenSettings Enum=TMGS_NoMipmaps, Texture Group:TextureGroup Enum=TEXTUREGROUP_UI) {is=title K2Node_FunctionEntry}
   > BeginTransaction(Context=ChangeIconSettings)
   > n1 = GetSelectedAssets()
   > n2 = For Each Loop(Array:Array of Object References=@n1) -> (Array Element:Object Reference)
   CheckoutLoadedAssets(execute=@n2.Completed, AssetsToCheckout=@n1)
   > SaveLoadedAssets(AssetsToSave=@n1, bOnlyIfIsDirty=true)
   > EndTransaction()
   n3 = Cast To Texture2D(execute=@n2.LoopBody, Object=@n2)
   > TransactObject(Object=@n3)
   > Set MipGenSettings!(
    execute
    MipGenSettings:TextureMipGenSettings Enum=Get Mip Gen Settings() -> (Mip Gen Settings:TextureMipGenSettings Enum=TMGS_FromTextureGroup)
    self:Texture Object Reference=@n3
   ) -> (then, Output_Get:TextureMipGenSettings Enum=TMGS_FromTextureGroup)
   > Set LODGroup!(execute, LODGroup:TextureGroup Enum=Get Texture Group() -> (Texture Group:TextureGroup Enum), self:Texture Object Reference=@n3) -> (then, Output_Get:TextureGroup Enum)
  ]}
 EventGraph {links [], nodes []}
]
variables []
