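# Editor utility action that bulk-fills season metadata on selected FortItemDefinition assets, then checks out and saves dirty ones within a transaction.
# It replaces only placeholders: Future release versions (also applied to pickaxe weapon definitions) and TBD names and descriptions.
# Descriptions are written through embedded Python using set_editor_property.
# The tag step only appends asset tags into the passed container; tag write-back is disconnected with a broken cast.
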EditorUtilityBlueprint /Game/ArtTools/HS_Season_Tool/HS_ScriptedActions/EUB_Action_MetaDataPass
graphs [
 EventGraph {links [], nodes []}
 MetaDataPass {nodes [
   MetaDataPass!() -> (
    then
    Fort Release Version:Fort Release Version Structure
    Item Name:Text
    Item Description:Text
    Tags:Owned Gameplay Tags Item Data Structure
    InOutTagContainer:"Gameplay Tag Container Structure (by ref)"
   ) {is=title K2Node_FunctionEntry}
   > BeginTransaction(Context=MetaDataPass)
   > n1 = GetSelectedAssets()
   > n2 = Filter by Class!(
    execute
    TargetArray:Array of Object References=@n1
    ObjectClass:Object Class Reference=/Script/FortniteGame.FortItemDefinition
    FilterType:EEditorScriptingFilterType Enum
   ) -> (then, ReturnValue:Array of Fort Item Definition Object References)
   > n3 = For Each Loop(Array:Array of Object References=@n2) -> (Array Element:Object Reference)
   CheckoutLoadedAssets(execute=@n3.Completed, AssetsToCheckout=@n2)
   > SaveLoadedAssets(AssetsToSave=@n2, bOnlyIfIsDirty=true)
   > EndTransaction()
   n4 = Cast To FortItemDefinition!(execute=@n3.LoopBody, Object=@n3) -> (then, CastFailed, AsFort Item Definition:Fort Item Definition Object Reference) {is=title K2Node_DynamicCast}
   > Set FortItemDefinition(FortItemDefinition:Fort Item Definition Object Reference=@n4) -> (Output_Get:Fort Item Definition Object Reference)
   > n5 = Sequence!(execute) -> (then_0, then_1, then_2:Exec, then_3:Exec)
   > n6 = EditorOnly_GetVersionSeasonNumber!(
    execute
    Version:"Fort Release Version Structure (by ref)"=Get Fort Release Version() -> (Fort Release Version:Fort Release Version Structure)
   ) -> (then, ReturnValue:int=0)
   > n7 = EditorOnly_GetItemDefinitionFortReleaseVersion!(
    execute
    ItemDefinition:Item Definition Base Object Reference=Get FortItemDefinition() -> (FortItemDefinition:Fort Item Definition Object Reference)
   ) -> (then, ReturnValue:Fort Release Version Structure)
   > n8 = EditorOnly_IsVersionEqualTo!(
    execute
    VersionA:"Fort Release Version Structure (by ref)"=@n7
    VersionB:"Fort Release Version Structure (by ref)"=Get FortReleaseVersion_Future() -> (FortReleaseVersion_Future:Fort Release Version Structure)
   ) -> (then, ReturnValue:bool)
   > n9 = Branch(Condition=AND Boolean(A=@n8, B="integer > integer"(A=@n6)))
   > SetFortReleaseVersion!(
    execute
    ItemDefinition:Fort Item Definition Object Reference=Get FortItemDefinition()
    FortReleaseVersion:"Fort Release Version Structure (by ref)"=Get Fort Release Version()
   ) -> (then, ReturnValue:bool)
   n10 = Get FortItemDefinition()
   > Branch(Condition="Equal (Class)"(A=Get Class(Object=@n10) -> (ReturnValue:Fort Item Definition Class Reference), B=/Script/FortniteGame.AthenaPickaxeItemDefinition))
   > n11 = Cast To AthenaPickaxeItemDefinition!(execute, Object=Get FortItemDefinition()) -> (then, CastFailed, AsAthena Pickaxe Item Definition:Athena Pickaxe Item Definition Object Reference) {is=title K2Node_DynamicCast}
   > SetFortReleaseVersion(
    ItemDefinition=Get WeaponDefinition!(self:Athena Pickaxe Item Definition Object Reference=@n11) -> (WeaponDefinition:Fort Weapon Melee Item Definition Object Reference)
    FortReleaseVersion=Get Fort Release Version()
   )
   LogString(execute=@n9.else, InString=Not Future version so dont change anything)
   n12 = Format Text(Format="LOCTEXT(\"TBD\")")
   n13 = Get ItemName!(self:Item Definition Base Object Reference=Get FortItemDefinition()) -> (ItemName:Text)
   n14 = TextIsEmpty(InText=Get Item Name() -> (Item Name:Text))
   n15 = Branch(
    execute=@n5.then_1
    Condition=AND Boolean(A="Equal Exactly (Text)"!(A:"Text (by ref)"=@n13, B:"Text (by ref)"=@n12) -> (ReturnValue:bool), B=NOT Boolean(A=@n14))
   )
   > SetRawDisplayName!(execute, self:Fort Item Definition Object Reference=Get FortItemDefinition(), InDisplayName:"Text (by ref)"=Get Item Name()) -> (then)
   LogString(execute=@n15.else, InString=name is not tbd so dont change anything)
   n16 = GetItemDescription!(self:Item Definition Base Object Reference=Get FortItemDefinition()) -> (ReturnValue:Text)
   n17 = Format Text(Format="LOCTEXT(\"TBD\")")
   n18 = TextIsEmpty(InText=Get Item Description() -> (Item Description:Text))
   Branch(execute=@n5.then_2, Condition=AND Boolean(A="Equal Exactly (Text)"(A=@n16, B=@n17), B=NOT Boolean(A=@n18)))
   > Execute Python Script!(
    execute
    PythonScript="import unreal\r\n\r\ndef set_item_description(obj_ref, text):\r\n    if not obj_ref:\r\n        unreal.log_warning(\"Object reference is None!\")\r\n        return\r\n\r\n    try:\r\n        obj_ref.set_editor_property(\"item_description\", text)\r\n        unreal.log(f\"Updated 'item_description' of {obj_ref.get_name()} to: {text}\")\r\n\r\n    except Exception as e:\r\n        unreal.log_error(f\"Error setting 'item_description': {e}\")\r\n\r\nitem_description = item_description\r\npath = path\r\nobj_reference = unreal.EditorAssetLibrary.load_asset(path)\r\nset_item_description(obj_reference, item_description)"
    item_description:Text=Get Item Description()
    path:String=GetSoftObjectPath!(Object=Get FortItemDefinition()) -> (ReturnValue_PathString:String)
   ) -> (then, ReturnValue) {inputs [item_description, path]}
   n19 = Cast To GET_TagContainer!(execute=@n5.then_3, Object=@n3) -> (then, CastFailed, AsGET Tag Container:GET Tag Container Object Reference) {is=title K2Node_DynamicCast}
   > AppendGameplayTagContainers!(
    execute
    InOutTagContainer:"Gameplay Tag Container Structure (by ref)"=Get InOutTagContainer() -> (InOutTagContainer:"Gameplay Tag Container Structure (by ref)")
    InTagContainer:"Gameplay Tag Container Structure (by ref)"=GetOwnedGameplayTags!(TagContainerInterface:Gameplay Tag Asset Interface Interface=@n19) -> (ReturnValue:Gameplay Tag Container Structure)
   ) -> (then)
   n20 = For Each Loop(Array:Array of Gameplay Tag Structures=GetOwnedGameplayTags!(TagContainerInterface) -> (ReturnValue_GameplayTags:Array of Gameplay Tag Structures)) -> (Array Element:Gameplay Tag Structure)
   > LogString(InString=Heddr99)
   EditorOnly_GetVersionSeasonNumber() -> (ReturnValue=0)
   GetTagContainer!(Value:"Value Or BBKey Gameplay Tag Container Structure (by ref)", BehaviorTreeComp:Behavior Tree Component Object Reference) -> (ReturnValue:Gameplay Tag Container Structure)
   MakeGameplayTagContainerFromArray!(GameplayTags:Array of Gameplay Tag Structures) -> (ReturnValue:Gameplay Tag Container Structure)
   GetDebugStringFromGameplayTag!(GameplayTag:Gameplay Tag Structure=@n20) -> (ReturnValue:String)
   GetOwnedGameplayTags!(TagContainerInterface=Get FortItemDefinition()) -> (ReturnValue:Gameplay Tag Container Structure)
   Set Owned Gameplay Tags Data!(
    execute
    self:Object Reference=Bad cast node!(Object:Wildcard) -> (bSuccess:bool=~, AsBuilder Tool Item Definition:Object Reference) {is=title K2Node_DynamicCast}.AsBuilder Tool Item Definition
    Data_Tags:Gameplay Tag Container Structure=Get InOutTagContainer()
   ) -> (then) {is=title K2Node_CallFunction}
   Execute Python Script!(execute, PythonScript="print('Hello World!')") -> (then, ReturnValue)
   Get TagContainer!(self:Fort Tooltip Context Object Reference) -> (TagContainer:Gameplay Tag Container Structure)
   Get GameplayTags!(self:Generic Gameplay Tag Container Object Reference) -> (GameplayTags:Gameplay Tag Container Structure)
   Get GameplayTags()
  ]}
]
variables [FortReleaseVersion_Future {type=FortReleaseVersion, default {versionName=Future}}]
