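# OpenWeaponID is an editor utility action that opens the weapon item definitions linked to selected pickaxe assets.
# It loops over the selected assets, casts each to AthenaPickaxeItemDefinition and gets the object path of its WeaponDefinition.
# An embedded Python script receives that path, loads the asset and opens its editor through AssetTools.
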EditorUtilityBlueprint /Game/ArtTools/HS_Season_Tool/HS_ScriptedActions/EUB_Action_OpenWeaponID
graphs [
 EventGraph {links [], nodes []}
 OpenWeaponID {nodes [
   OpenWeaponID() {is=title K2Node_FunctionEntry}
   > n1 = GetSelectedAssets()
   > n2 = For Each Loop(Array:Array of Object References=@n1) -> (Array Element:Object Reference)
   > n3 = Cast To AthenaPickaxeItemDefinition!(execute, Object=@n2) -> (then, CastFailed, AsAthena Pickaxe Item Definition:Athena Pickaxe Item Definition Object Reference) {is=title K2Node_DynamicCast}
   > Execute Python Script!(
    execute
    PythonScript="import unreal\r\n\r\nasset_path = asset_path\r\nasset = unreal.EditorAssetLibrary.load_asset(asset_path)\r\n\r\nif asset:\r\n    unreal.AssetToolsHelpers.get_asset_tools().open_editor_for_assets([asset])"
    asset_path:String=Get Object Path String(Object=Get WeaponDefinition!(self:Athena Pickaxe Item Definition Object Reference=@n3) -> (WeaponDefinition:Fort Weapon Melee Item Definition Object Reference))
   ) -> (then, ReturnValue) {inputs [asset_path]}
  ]}
]
variables []
