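# Editor utility that batch-applies pickup presets to selected static meshes, then checks out and saves them.
# SetValuesBasedOnType loads LOD screen sizes and reduction settings from DT_PickupPresets, using the pickup type as row name.
# Inside a transaction, each mesh gets LOD group, per-quality minimum LODs, screen sizes, collision LOD, cleared collisions with a trace flag, and Nanite settings.
# The reduction-settings nodes are unwired, so those preset values go unapplied.
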EditorUtilityBlueprint /Game/ArtTools/HS_Season_Tool/HS_ScriptedActions/EUB_Action_PickupSettings
graphs [
 ApplyPickupSettings {nodes [
   n1 = ApplyPickupSettings!() -> (then, PickupType:Enum_PickupTypeImplementation Enum=Enum_PickupTypeImplementation_MAX) {is=title K2Node_FunctionEntry}
   > SetValuesBasedOnType!(execute, self, Pickup_Type:Enum_PickupTypeImplementation Enum=@n1) -> (then)
   > BeginTransaction(Context=ChangePickupSettings)
   > n2 = GetSelectedAssets()
   > n3 = For Each Loop(Array:Array of Object References=@n2) -> (Array Element:Object Reference)
   CheckoutLoadedAssets(execute=@n3.Completed, AssetsToCheckout=@n2)
   > SaveLoadedAssets(AssetsToSave=@n2, bOnlyIfIsDirty=false)
   > EndTransaction()
   n4 = Cast To StaticMesh(execute=@n3.LoopBody, Object=@n3)
   > Set L_StaticMeshElement(L_StaticMeshElement:Static Mesh Object Reference=@n4) -> (Output_Get:Static Mesh Object Reference)
   > TransactObject(Object=Get L_StaticMeshElement() -> (L_StaticMeshElement:Static Mesh Object Reference))
   > n5 = Sequence!(execute) -> (then_0, then_1, then_2:Exec, then_3:Exec, then_4:Exec, then_5:Exec)
   n6 = Get L_StaticMeshElement()
   n7 = GetNumLODs(self=@n6)
   n8 = Branch(execute=@n5.then_2, Condition="integer >= integer"(A=@n7, B=3))
   > Set LODForCollision!(execute, LODForCollision:int=2, self:Static Mesh Object Reference=@n6) -> (then, Output_Get:int=0)
   Set LODForCollision(execute=@n8.else, LODForCollision=int - int(A=@n7), self=@n6) -> (Output_Get=0)
   SetLodScreenSizes!(
    execute=@n5.then_1
    self:Static Mesh Editor Subsystem Object Reference=Get StaticMeshEditorSubsystem()
    StaticMesh:Static Mesh Object Reference=Get L_StaticMeshElement()
    ScreenSizes:"Array of Float (single-precision)s"=Get LOD_Screen_Size() -> (LOD_Screen_Size:"Array of Float (double-precision)s")
   ) -> (then, ReturnValue:bool)
   n9 = Get L_StaticMeshElement()
   RemoveCollisions!(
    execute=@n5.then_3
    self:Static Mesh Editor Subsystem Object Reference=Get StaticMeshEditorSubsystem()
    StaticMesh:Static Mesh Object Reference=@n9
   ) -> (then, ReturnValue:bool)
   > n10 = GetEditorProperty!(execute:Exec, Object:Object Reference=@n9, PropertyName:Name=BodySetup) -> (then:Exec, PropertyValue:Object Reference, ReturnValue:bool)
   > SetEditorProperty(
    Object=@n10
    PropertyName=CollisionTraceFlag
    PropertyValue:ECollisionTraceFlag Enum=Get L_CollisionTraceFlag() -> (L_CollisionTraceFlag:ECollisionTraceFlag Enum=CTF_UseDefault)
   )
   n11 = GetLODGroup!(
    execute=@n5.then_0
    self:Static Mesh Editor Subsystem Object Reference=Get StaticMeshEditorSubsystem()
    StaticMesh:Static Mesh Object Reference=Get L_StaticMeshElement()
   ) -> (then, ReturnValue:Name=None)
   > n12 = Branch(Condition="Equal (Name)"(A=@n11, B=Get L_PresetLOD_Group() -> (L_PresetLOD_Group:Name=None)))
   SetMinLODForQualityLevels!(
    execute=[@n13.then, @n12.then]
    self:Static Mesh Object Reference=Get L_StaticMeshElement()
    QualityLevelMinimumLODs:Map of EPerQualityLevels Enums to Integers=Get L_Env_LOD_Quality_Defaults() -> (L_Env_LOD_Quality_Defaults:Map of EPerQualityLevels Enums to Integers=Low)
    Default:int=-1
   ) -> (then)
   n13 = SetLODGroup!(
    execute=@n12.else
    self:Static Mesh Editor Subsystem Object Reference=Get StaticMeshEditorSubsystem()
    StaticMesh:Static Mesh Object Reference=Get L_StaticMeshElement()
    LODGroup:Name=Get L_PresetLOD_Group() -> (L_PresetLOD_Group=None)
    bRebuildImmediately:bool=true
   ) -> (then, ReturnValue:bool)
   SetNaniteSettings!(
    execute=@n5.then_4
    self:Static Mesh Editor Subsystem Object Reference=Get StaticMeshEditorSubsystem()
    StaticMesh:Static Mesh Object Reference=Get L_StaticMeshElement()
    NaniteSettings:Mesh Nanite Settings Structure=Get STR Env Nanite Settings() -> (STR Env Nanite Settings:Mesh Nanite Settings Structure)
    bApplyChanges:bool=true
   ) -> (then)
   n14 = Get L_StaticMeshElement()
   Branch(Condition="integer >= integer"(A=GetNumLODs(self=@n14), B=3))
   > SetLodReductionSettings!(
    execute
    self:Static Mesh Editor Subsystem Object Reference=Get StaticMeshEditorSubsystem()
    StaticMesh:Static Mesh Object Reference=Get L_StaticMeshElement()
    LodIndex:int=0
    ReductionOptions:"Mesh Reduction Settings Structure (by ref)"=Get LOD0_MeshReductionSettings() -> (LOD0_MeshReductionSettings:Mesh Reduction Settings Structure)
   ) -> (then)
   n15 = Get StaticMeshEditorSubsystem()
   > SetLodReductionSettings(
    self=@n15
    StaticMesh=Get L_StaticMeshElement()
    LodIndex=1
    ReductionOptions=Get LOD1_MeshReductionSettings() -> (LOD1_MeshReductionSettings:Mesh Reduction Settings Structure)
   )
   > SetLodReductionSettings(
    self=Get StaticMeshEditorSubsystem()
    StaticMesh=Get L_StaticMeshElement()
    LodIndex=2
    ReductionOptions=Get LOD2_MeshReductionSettings() -> (LOD2_MeshReductionSettings:Mesh Reduction Settings Structure)
   )
   SetLODGroup(LODGroup=None, bRebuildImmediately=true)
   SetLods!(
    execute
    self:Static Mesh Editor Subsystem Object Reference=@n15
    StaticMesh:Static Mesh Object Reference
    ReductionOptions:"Static Mesh Reduction Options Structure (by ref)"
   ) -> (then, ReturnValue:int=0)
  ]}
 EventGraph {links [], nodes []}
 SetValuesBasedOnType {nodes [
   n1 = SetValuesBasedOnType!() -> (then, Pickup_Type:Enum_PickupTypeImplementation Enum=NewEnumerator0) {is=title K2Node_FunctionEntry}
   n2 = Enum to String!(Enumerator:Enum_PickupTypeImplementation Enum=@n1) -> (ReturnValue:String)
   > n3 = Get Data Table Row DT_PickupPresets!(
    execute:Exec
    DataTable:Data Table Object Reference=&/DT_PickupPresets
    RowName:Name=String To Name(InString=Replace(SourceString=@n2, From=" - "))
   ) -> (then:Exec, RowNotFound:Exec, ReturnValue:Pickup Presets Structure Structure) {is=title K2Node_GetDataTableRow}
   n4 = Break Pickup Presets Structure!(Pickup_Presets_Structure:"Pickup Presets Structure Structure (by ref)"=@n3) -> (
    LODScreenSizes_18_#1:"Array of Float (double-precision)s"
    NumberofVerticesToConsiderSize_19_#2:int
    LOD0-MeshReductionSettings_22_#3:Mesh Reduction Settings Structure
    LOD1-MeshReductionSettings_26_#4:Mesh Reduction Settings Structure
    LOD2-MeshReductionSettings_28_#5:Mesh Reduction Settings Structure
   ) {is=title K2Node_BreakStruct}
   > Set LOD_Screen_Size(LOD_Screen_Size:"Array of Float (double-precision)s"=@n4) -> (Output_Get:"Array of Float (double-precision)s")
   > Set LOD0_MeshReductionSettings(LOD0_MeshReductionSettings:Mesh Reduction Settings Structure=@n4.LOD0-MeshReductionSettings_22_#3) -> (Output_Get:Mesh Reduction Settings Structure)
   > Set LOD1_MeshReductionSettings(LOD1_MeshReductionSettings:Mesh Reduction Settings Structure=@n4.LOD1-MeshReductionSettings_26_#4) -> (Output_Get:Mesh Reduction Settings Structure)
   > Set LOD2_MeshReductionSettings(LOD2_MeshReductionSettings:Mesh Reduction Settings Structure=@n4.LOD2-MeshReductionSettings_28_#5) -> (Output_Get:Mesh Reduction Settings Structure)
  ]}
]
variables [
 LOD0_MeshReductionSettings {
  type=MeshReductionSettings
  default {
   bCullOccluded=false
   bGenerateUniqueLightmapUVs=false
   bKeepSymmetry=false
   bRecalculateNormals=false
   bVisibilityAided=false
   baseLODModel=0
   hardAngleThreshold=80
   maxDeviation=0
   maxNumOfTriangles=4294967295
   maxNumOfVerts=4294967295
   percentTriangles=1
   percentVertices=1
   pixelError=8
   shadingImportance=Normal
   silhouetteImportance=Normal
   terminationCriterion=Triangles
   textureImportance=Normal
   vertexColorImportance=Off
   visibilityAggressiveness=Lowest
   weldingThreshold=0
  }
 }
 LOD1_MeshReductionSettings {
  type=MeshReductionSettings
  default {
   bCullOccluded=false
   bGenerateUniqueLightmapUVs=false
   bKeepSymmetry=false
   bRecalculateNormals=false
   bVisibilityAided=false
   baseLODModel=0
   hardAngleThreshold=80
   maxDeviation=0
   maxNumOfTriangles=4294967295
   maxNumOfVerts=4294967295
   percentTriangles=1
   percentVertices=1
   pixelError=8
   shadingImportance=Normal
   silhouetteImportance=Normal
   terminationCriterion=Triangles
   textureImportance=Normal
   vertexColorImportance=Off
   visibilityAggressiveness=Lowest
   weldingThreshold=0
  }
 }
 LOD2_MeshReductionSettings {
  type=MeshReductionSettings
  default {
   bCullOccluded=false
   bGenerateUniqueLightmapUVs=false
   bKeepSymmetry=false
   bRecalculateNormals=false
   bVisibilityAided=false
   baseLODModel=0
   hardAngleThreshold=80
   maxDeviation=0
   maxNumOfTriangles=4294967295
   maxNumOfVerts=4294967295
   percentTriangles=1
   percentVertices=1
   pixelError=8
   shadingImportance=Normal
   silhouetteImportance=Normal
   terminationCriterion=Triangles
   textureImportance=Normal
   vertexColorImportance=Off
   visibilityAggressiveness=Lowest
   weldingThreshold=0
  }
 }
 LOD_Screen_Size {type="array<number>", default [1,0.4,0.1,0.03]}
 STR Env Nanite Settings {
  type=MeshNaniteSettings
  default {
   bEnabled=true
   bExplicitTangents=false
   bLerpUVs=true
   bSeparable=false
   bVoxelNDF=true
   bVoxelOpacity=false
   boneWeightPrecision=-1
   displacementUVChannel=0
   fallbackPercentTriangles=1
   fallbackRelativeError=0
   fallbackTarget=RelativeError
   generateFallback=PlatformDefault
   keepPercentTriangles=1
   maxEdgeLengthFactor=0
   normalPrecision=-1
   numRays=1
   positionPrecision=-2147483648
   rayBackUp=0
   shapePreservation=None
   tangentPrecision=-1
   targetMinimumResidencyInKB=0
   trimRelativeError=0
   voxelLevel=0
   displacementMaps []
   naniteAssemblyData {nodes [], parts []}
  }
 }
]
