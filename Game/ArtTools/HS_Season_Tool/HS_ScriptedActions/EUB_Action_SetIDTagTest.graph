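# Editor utility debug test for retagging selected Fort item definitions, likely for the season tool.
# One chain copies each item's owned-tags data, removes a hardcoded Season 34 cosmetic filter tag, appends TagsToAddContainer, and writes it back.
# After the loop it checks out and saves dirty assets.
# An unfinished transaction-based variant and stray nodes remain, and the function entry connects to neither chain.
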EditorUtilityBlueprint /Game/ArtTools/HS_Season_Tool/HS_ScriptedActions/EUB_Action_SetIDTagTest
graphs [
 DEBUG_SetID_Tag_Test {nodes [
   n1 = GetSelectedAssets()
   > n2 = For Each Loop(Array:Array of Object References=@n1) -> (Array Element:Object Reference)
   CheckoutLoadedAssets(execute=@n2.Completed, AssetsToCheckout=@n1)
   > SaveLoadedAssets(AssetsToSave=@n1, bOnlyIfIsDirty=true)
   n3 = Cast To FortItemDefinition!(execute=@n2.LoopBody, Object=@n2) -> (then, CastFailed, AsFort Item Definition:Fort Item Definition Object Reference) {is=title K2Node_DynamicCast}
   > n4 = Sequence!(execute) -> (then_0, then_1, then_2:Exec)
   > n5 = GetItemData!(execute, self:Item Definition Base Object Reference=@n3) -> (Found:Exec, NotFound:Exec, ItemData:Owned Gameplay Tags Item Data Structure)
   > Set L Item Data Copy(L Item Data Copy:Owned Gameplay Tags Item Data Structure=@n5) -> (Output_Get:Owned Gameplay Tags Item Data Structure)
   Return Node(execute=@n5.NotFound)
   n6 = Sequence!(execute=@n4.then_1) -> (then_0, then_1)
   > n7 = Set L Container(L Container:Gameplay Tag Container Structure=Break Owned Gameplay Tags Item Data!(ItemComponentData_OwnedGameplayTags:"Owned Gameplay Tags Item Data Structure (by ref)"=Get L Item Data Copy() -> (L Item Data Copy:Owned Gameplay Tags Item Data Structure)) -> (Tags:Gameplay Tag Container Structure) {is=title K2Node_BreakStruct}) -> (Output_Get:Gameplay Tag Container Structure)
   > RemoveGameplayTag!(execute, TagContainer:"Gameplay Tag Container Structure (by ref)"=@n7, Tag:Gameplay Tag Structure=(TagName="Cosmetics.Filter.Season.34")) -> (then, ReturnValue:bool)
   > Set members in Owned Gameplay Tags Item Data!(
    execute:Exec
    StructRef:"Owned Gameplay Tags Item Data Structure (by ref)"=Get L Item Data Copy()
    Tags:Gameplay Tag Container Structure=Get L Container() -> (L Container:Gameplay Tag Container Structure)
   ) -> (then:Exec, StructOut:"Owned Gameplay Tags Item Data Structure (by ref)") {is=title K2Node_SetFieldsInStruct}
   n8 = Set L Container(execute=@n6.then_1, L Container=Break Owned Gameplay Tags Item Data(ItemComponentData_OwnedGameplayTags=Get L Item Data Copy()))
   > AppendGameplayTagContainers!(
    execute
    InOutTagContainer:"Gameplay Tag Container Structure (by ref)"=@n8
    InTagContainer:"Gameplay Tag Container Structure (by ref)"=Get TagsToAddContainer() -> (TagsToAddContainer:Gameplay Tag Container Structure)
   ) -> (then)
   > Set members in Owned Gameplay Tags Item Data(StructRef=Get L Item Data Copy(), Tags=Get L Container())
   SetItemData!(execute=@n4.then_2, self:Item Definition Base Object Reference=@n3, ItemData:Owned Gameplay Tags Item Data Structure=Get L Item Data Copy()) -> (then)
   BeginTransaction(Context=SetTagToID)
   > n9 = GetSelectedAssets()
   > n10 = For Each Loop(Array=@n9)
   CheckoutLoadedAssets(execute=@n10.Completed, AssetsToCheckout=@n9)
   > SaveLoadedAssets(AssetsToSave=@n9, bOnlyIfIsDirty=true)
   > EndTransaction()
   n11 = Cast To FortItemDefinition(execute=@n10.LoopBody, Object=@n10)
   > TransactObject(Object=@n11)
   n12 = GetOwnedGameplayTags!(TagContainerInterface:Gameplay Tag Asset Interface Interface=@n11) -> (ReturnValue:Gameplay Tag Container Structure)
   AddTag!(
    execute
    Container:"Gameplay Tag Container Structure (by ref)"=@n12
    GameplayTag:"Gameplay Tag Structure (by ref)"=Get GameplayTag Container() -> (GameplayTag Container:Gameplay Tag Structure)
   ) -> (then)
   > AddTag(Container=@n12, GameplayTag=Get GameplayTag() -> (GameplayTag:Gameplay Tag Structure))
   LogString(InString=Hello)
   AddTag()
   DEBUG_SetID_Tag_Test!() -> (then, TagsToAddContainer:Gameplay Tag Container Structure) {is=title K2Node_FunctionEntry}
   Get OwnedGameplayTagItemData() -> (OwnedGameplayTagItemData:Owned Gameplay Tags Item Data Structure)
  ]}
 EventGraph {links [], nodes []}
]
variables []
